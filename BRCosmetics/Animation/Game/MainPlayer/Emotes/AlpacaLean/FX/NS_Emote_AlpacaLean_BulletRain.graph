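# Niagara system for the AlpacaLean emote that rains tumbling mesh bullets with sprites around the owner, culling spawns near the center.
# Bullets fall along curve-driven lerps, ray-trace downward, then bounce and settle on hits, with front-end flags forcing landings.
# Impacts are written to a position array that system-level ReadDeath uses to spawn FlashFlares and Sparks.
# FakeBullets add falling filler without collision tracing, looping each particle's lifetime.
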NiagaraSystem /BRCosmetics/Animation/Game/MainPlayer/Emotes/AlpacaLean/FX/NS_Emote_AlpacaLean_BulletRain
dependencies {
 usedDataInterfaces [
  &NiagaraDataInterfaceCurve
  &NiagaraDataInterfaceColorCurve
  &NiagaraDataInterfaceSpriteRendererInfo
  &NiagaraDataInterfaceVector2DCurve
  &NiagaraDataInterfaceCollisionQuery
  &NiagaraDataInterfaceDebugDraw
  &NiagaraDataInterfaceVectorCurve
  &NiagaraDataInterfaceCamera
  &NiagaraDataInterfaceArrayPosition
 ]
 usedDynamicInputs [&Lerp_Float, &ReturnCameraVector]
 usedModules [
  &self:SystemSpawnScript.SystemScriptSource.SystemScriptGraph.NiagaraNodeAssignment_2.SetVariables_#1
  &SystemState
  &self:SystemSpawnScript.SystemScriptSource.SystemScriptGraph.NiagaraNodeAssignment_0.SetVariables_#2
  &self:ReadDeath
  &self:SystemSpawnScript.SystemScriptSource.SystemScriptGraph.NiagaraNodeAssignment_3.SetVariables_#3
  &self:SystemSpawnScript.SystemScriptSource.SystemScriptGraph.NiagaraNodeAssignment_4.SetVariables_#4
 ]
}
emitters {
 Bullets {
  emitterUpdateScript [
   &EmitterState(Life Cycle Mode=::NewEnumerator1, Loop Duration=2.05, UseLoopDelay=false)
   &SpawnRate(SpawnRate=30, Use Spawn Probability=false)
  ]
  particleSpawnScript [
   /SetVariables(Particles.RandomHeightOffset:NiagaraFloat=&RandomRangeFloat(Minimum=-20, Maximum=20)) {isSetParameters=true}
   &InitializeParticle(
    Lifetime=&RandomRangeFloat(Minimum=1.6, Maximum=2.6)
    Mass Mode=::NewEnumerator1
    Mesh Scale Mode=::NewEnumerator3
    Position Offset=&MakeVector(X=0, Y=0, Z=&Add_Float(A=$Particles.RandomHeightOffset, B=$User.HeightStart))
    Sprite Rotation Angle=-0.25
    Sprite Rotation Mode=::NewEnumerator1
    Sprite Size=&RandomRangeVector2D(Minimum=<80 2>, Maximum=<120 3>)
    Sprite Size Mode=::NewEnumerator4
    UsePositionOffset=true
    Write Material Random=true
   )
   &ShapeLocation(Cylinder Height=20, MaskSpawnGroup=false, Override Seed=false, Shape Primitive=::NewEnumerator1, Surface Only=false, UseOwnerScale=true)
   /SetVariables(Particles.RandomDirection:Vector3f=&Add_Vector(A=<0 0 0>, B=&RandomRangeVector(Minimum=<-0.1 -0.1 0>, Maximum=<0.1 0.1 0>))) {isSetParameters=true, module=SetVariables_#5}
   /SetVariables() {isSetParameters=true, module=SetVariables_#6}
   /SetVariables() {isSetParameters=true}
   /SetVariables(Particles.StartPosition:NiagaraPosition=&Subtract_Position(A=$Particles.Position, B=/Niagara/DynamicInputs/Helpers/SimulationPosition())) {isSetParameters=true, module=SetVariables_#7}
   /SetVariables() {isSetParameters=true, module=SetVariables_#8}
   /SetVariables(Particles.MeshOrientation=&MakeQuaternion(
     Quaternion From=5
     Axis=?
     Angle=?
     Yaw=?
     Pitch=?
     Roll=?
     X=?
     Y=?
     Z=?
     From=?
     To=?
     Stabilization Mode:ENiagara_OrientVectorCount=::NewEnumerator0
     Side Direction:Vector3f=?
     Up Direction:Vector3f=?
     Angle Type=::NewEnumerator0
     Coordinate Space=?
     Normalize:NiagaraBool=true
    )) {isSetParameters=true}
   /SetVariables(Particles.NearPlayer:NiagaraBool=&SetBoolByFloatComparison(
     A=&VectorLength(Vector=&Multiply_Vector(
       A=&TransformPosition(Position=$Particles.Position, Source Space=::Simulation, Destination Space=::Local)
       B=<1 1 0>
      ))
     B=40
     Comparison Type=::NewEnumerator6
    )) {isSetParameters=true}
   &KillParticles(Kill Particles=$Particles.NearPlayer) {module=KillParticles001}
  ]
  particleUpdateScript [
   &ParticleState(Kill Particles When Lifetime Has Elapsed=false)
   /SetVariables(
    Particles.CollisionValid:NiagaraBool=&Boolean_OrOperation(
     Bool1=&Boolean_AndOperation(
      Bool1=&Boolean_OrOperation(Bool1=$User.bIsFrontEndPreview, Bool2=$User.bIsFrontEnd)
      Bool2=&SetBoolByFloatComparison(A=$Particles.NormalizedAge, B=0.5)
     )
     Bool2=$Particles.CollisionValid
    )
    Particles.DoneOnce:NiagaraBool=&Boolean_OrOperation(
     Bool1=&SetBoolByFloatComparison(A=&VectorLength(Vector=$Engine.Owner.Velocity), B=500)
     Bool2=$Particles.DoneOnce
    )
   ) {isSetParameters=true}
   /SetVariables(Particles.CollisionPos:NiagaraPosition=&MakeCustomVectorFromBool(
     Bool=$Particles.CollisionValid
     True Vector=Linked() {linkedVariable {name=Particles.CollisionPos, type=NiagaraPosition}}
     False Vector=&AddVectorToPosition(
      Position=&AddVectorToPosition(Position=$Particles.Initial.StartPosition, Vector=&SimulationPosition())
      Vector=&Multiply_VectorByFloat(
       Vector=<0 0 -1>
       Float=&Add_Float(A=$Particles.RandomHeightOffset, B=&Multiply_Float(A=$User.HeightStart, B=1))
      )
     )
    )) {isSetParameters=true, module=SetVariables_#9}
   &self:RayTraceCollision(
    Disable Trace:NiagaraBool=&Boolean_OrOperation(
     Bool1=&SetBoolByFloatComparison(A=$Particles.NormalizedAge, B=$User.CollisionCheck, Comparison Type=::NewEnumerator6)
     Bool2=&Boolean_OrOperation(Bool1=$Particles.CollisionValid, Bool2=$Particles.DoneOnce)
    )
    DrawDebug:NiagaraBool=&Boolean_AndOperation(
     Bool1=false
     Bool2=&SetBoolByFloatComparison(A=$Particles.NormalizedAge, B=0.5, Comparison Type=::NewEnumerator6)
    )
    TraceDir:Vector3f=<0 0 -1>
    TraceDist=120
    bIsFrontEnd:NiagaraBool=$User.bIsFrontEndPreview
   ) {module=RayTraceCollision}
   /SetVariables(Particles.SecondStartingPoint:NiagaraPosition=/Niagara/DynamicInputs/TypeConversions/MakeCustomPositionFromBool(
     Bool:NiagaraBool=$Particles.CollisionValid
     True Position:NiagaraPosition=$Particles.SecondStartingPoint
     False Position:NiagaraPosition=$Particles.Position
    )) {isSetParameters=true}
   /SetVariables(Particles.Position=&AddVectorToPosition(
     Position=&MakeCustomPositionFromBool(
      Bool:NiagaraBool=$Particles.CollisionValid
      True Position:NiagaraPosition=&LerpPosition(
       A=$Particles.SecondStartingPoint
       B=$Particles.CollisionPos
       Alpha=&FloatFromCurve(
        FloatCurve=DataInterface() {propertyValues {Curve {defaultValue=1, keys [{arriveTangent=3.9126, time=0, value=0}, {leaveTangent=4.8076, time=1, value=1}]}}}
        CurveIndex=&Divide_Float(
         A=&Subtract_Float(A=$Particles.NormalizedAge, B=$User.CollisionCheck)
         B=&Subtract_Float(A=0.5, B=$User.CollisionCheck)
        )
       )
      )
      False Position:NiagaraPosition=&LerpPosition(
       A=&AddVectorToPosition(
        Position=&SimulationPosition()
        Vector=Linked() {linkedVariable {name=Particles.Initial.StartPosition, type=NiagaraPosition}}
       )
       B=$Particles.CollisionPos
       Alpha=&FloatFromCurve(
        FloatCurve=DataInterface() {propertyValues {Curve {
           defaultValue=1
           keys [
            {arriveTangent=0.7283, interpMode=RCIM_Cubic, time=0, value=0}
            {arriveTangent=3.2194, interpMode=RCIM_Cubic, leaveTangent=3.2194, tangentMode=RCTM_User, time=0.5, value=1}
           ]
          }}}
        CurveIndex=&Multiply_Float(A=$Particles.NormalizedAge, B=1)
       )
      )
     )
     Vector=&MakeCustomVectorFromBool(
      Bool=$Particles.CollisionValid
      True Vector=&Add_Vector(
       A=<0 0 2>
       B=&Add_Vector(
        A=&Lerp_Vector(
         A=<0 0 0>
         B=<0 0 20>
         Alpha=&FloatFromCurve(FloatCurve=DataInterface() {propertyValues {Curve {keys [
              <at 0.5 0>
              <at 0.6012 0.9948>
              {arriveTangent=-1.698, leaveTangent=-1.698, time=0.761, value=0.0242}
              <at 0.8268 0.5586>
              {arriveTangent=-3.3303, leaveTangent=-3.3303, time=0.8971, value=-0.0121}
              <at 0.9487 0.2484>
              <at 1 0>
             ]}}})
        )
        B=&Multiply_VectorByFloat(
         Vector=&RotateVectorByQuaternion(
          Vector To Rotate=&NormalizeVector(
           Vector To Normalize=&Multiply_Vector(A=$Particles.Initial.RandomDirection, B=<1 1 0>)
           Safe Normalize=true
           Safe Fallback Threshold=0
           Safe Fallback Vector=<0 0 0.001>
          )
          Quaternion=&MakeQuaternion(
           Quaternion From=3
           Axis=<1 0 0>
           Angle=0
           From=<0 0 1>
           To=$Particles.CollisionNormal
           Stabilization Mode:ENiagara_OrientVectorCount=::NewEnumerator0
           Side Direction:Vector3f=?
           Up Direction:Vector3f=?
           Angle Type=::NewEnumerator0
           Normalize:NiagaraBool=true
          )
         )
         Float=&FloatFromCurve(
          FloatCurve=DataInterface() {propertyValues {Curve {keys [{arriveTangent=1.4069, interpMode=RCIM_Cubic, leaveTangent=1.4069, tangentMode=RCTM_User, time=0.5, value=0}, <at 2 1 RCIM_Cubic>]}}}
          Scale Curve=100
         )
        )
       )
      )
      False Vector=&Multiply_VectorByFloat(
       Vector=<0 0 -1>
       Float=&FloatFromCurve(FloatCurve=DataInterface() {propertyValues {Curve {keys [<at 0.5 0>,<at 1 1>]}}}, Scale Curve=200)
      )
     )
    )) {isSetParameters=true, module=SetVariables_#10}
   &self:WriteDeath(
    Position Array:NiagaraDataInterfaceArrayPosition=$System.DeathArray
    bool:NiagaraBool=&Boolean_OrOperation(
     Bool1=&InvertBool(Boolean:NiagaraBool=$Particles.CollisionValid)
     Bool2=&Boolean_OrOperation(
      Bool1=&SetBoolByFloatComparison(A=$Particles.NormalizedAge, B=0.5, Comparison Type=::NewEnumerator7)
      Bool2=&SetBoolByFloatComparison(
       A=$Particles.NormalizedAge
       B=&Add_Float(A=0.5, B=&Multiply_Float(A=$Engine.WorldDeltaTime, B=0.5))
      )
     )
    )
   ) {module=WriteDeath}
   &UpdateMeshOrientation(
    Facing Direction=&RandomRangeVector(Minimum=<-1 -1 -1>, Maximum=<1 1 1>)
    Facing Interpolation=false
    Orientation Method=::NewEnumerator1
    Override Quat=false
    TransformFacing=false
    TransformUp=false
    Write to Orientation:ENiagara_AttributeSamplingApplyOutput=::NewEnumerator1
   ) {module=UpdateMeshOrientation001}
   &UpdateMeshOrientation(
    Rotation Coordinate Space=::NewEnumerator0
    Rotation Rate=&FloatFromCurve(
     FloatCurve=DataInterface() {propertyValues {Curve {keys [<at 0 0.5 RCIM_Cubic>,<at 0.5 1 RCIM_Cubic>]}}}
     Scale Curve=&RandomRangeFloat(Minimum=0.5, Maximum=1.5)
    )
    Rotation Vector=&RandomRangeVector(Minimum=<-1 -1 -1>, Maximum=<1 1 1>)
    Write to Orientation:ENiagara_AttributeSamplingApplyOutput=::NewEnumerator1
   ) {module=UpdateMeshOrientation004}
   &UpdateMeshOrientation(
    Rotation Rate=&FloatFromCurve(FloatCurve=DataInterface() {propertyValues {Curve {keys [<at 0.5 0>,<at 0.6 1>]}}}, Scale Curve=3)
    Rotation Vector=&RandomRangeVector(Minimum=<-1 -1 -1>, Maximum=<1 1 1>)
    Write to Orientation:ENiagara_AttributeSamplingApplyOutput=::NewEnumerator1
   )
   &UpdateMeshOrientation(
    Facing Direction=/Niagara/DynamicInputs/Vectors/CrossProduct(
     Vector A:Vector3f=&NormalizeVector(
      Vector To Normalize=&Multiply_Vector(A=<1 1 0>, B=$Particles.Initial.RandomDirection)
      Safe Normalize=true
      Safe Fallback Threshold=0
      Safe Fallback Vector=<0 0 0.001>
     )
     Vector B:Vector3f=<0 0 1>
     Normalize Incoming Vectors:NiagaraBool=true
    )
    Facing Interpolation=false
    Orientation Method=::NewEnumerator1
    Override Incoming Orientation=1
    Override Quat=true
    Override Quaternion:Quat4f=$Particles.BaseQuat
    TransformFacing=false
    TransformUp=false
    Write to Orientation:ENiagara_AttributeSamplingApplyOutput=::NewEnumerator1
   ) {module=UpdateMeshOrientation003}
   /SetVariables(Particles.BaseQuat:Quat4f=&MultiplyQuaternion(
     Quaternion A=$Output.UpdateMeshOrientation003.MeshOrientation
     Quaternion B=&MakeQuaternion(
      Quaternion From=1
      Axis=<1 0 0>
      Angle=0
      Pitch=&FloatFromCurve(
       FloatCurve=DataInterface() {propertyValues {Curve {keys [<at 1 1 RCIM_Cubic>,<at 1.5 0 RCIM_Cubic>]}}}
       Scale Curve=&RandomRangeFloat(Minimum=2, Maximum=5)
      )
      Roll=30
      Y=5
      Z=30
      From=?
      To=?
      Stabilization Mode:ENiagara_OrientVectorCount=::NewEnumerator0
      Side Direction:Vector3f=?
      Up Direction:Vector3f=?
      Angle Type=::NewEnumerator0
      Normalize:NiagaraBool=true
     )
    )) {isSetParameters=true}
   &ScaleMeshSize(Scale Factor=&Multiply_Vector(
     A=&VectorFromCurve(VectorCurve=DataInterface() {propertyValues {
        XCurve {keys [<at 0 1>,<at 1 1>,<at 1 1>]}
        YCurve {keys [<at 0 1 RCIM_Cubic>, <at 0.45 1.3 RCIM_Cubic>, {arriveTangent=-1.6667, time=0.5, value=1}, <at 1 1 RCIM_Cubic>, <at 1 1 RCIM_Cubic>]}
        ZCurve {keys [<at 0 1>,<at 1 1>,<at 1 1>]}
       }})
     B=&VectorFromFloat(Value=&FloatFromCurve(
       FloatCurve=DataInterface() {propertyValues {Curve {keys [
           <at 0 0 RCIM_Cubic>
           <at 0.1 1>
           {arriveTangent=-0.625, interpMode=RCIM_Cubic, leaveTangent=-0.625, time=1.2, value=1}
           <at 1.7 0 RCIM_Cubic>
          ]}}}
       Scale Curve=&Multiply_Float(
        A=/Niagara/DynamicInputs/Camera/CameraFade(
         Camera Query:NiagaraDataInterfaceCamera=DataInterface()
         Fade Offset:NiagaraFloat=50
         Fade Distance:NiagaraFloat=100
         Use Owner Position:NiagaraBool=true
        )
        B=&MakeCustomFloatFromBool(
         Bool=$Particles.CollisionValid
         True Float=1
         False Float=&FloatFromCurve(FloatCurve=DataInterface() {propertyValues {Curve {keys [<at 0.5 1 RCIM_Cubic>,<at 1 0 RCIM_Cubic>]}}})
        )
       )
      ))
    )) {module=ScaleMeshSize001}
   /SetVariables(Particles.StateIndex:NiagaraInt32=&MakeIntFromFloat(
     Float=&FloatFromCurve(FloatCurve=DataInterface() {propertyValues {Curve {keys [<at 0 0 RCIM_Constant>,<at 0.5 1 RCIM_Constant>]}}})
     Conversion Mode=::NewEnumerator1
    )) {isSetParameters=true, module=SetVariables_#11}
   /SetVariables(Particles.MeshOrientation=&LerpQuaternion(
     Quaternion A:Quat4f=/Game/Effects/Niagara/Modules/MergeEmitters/MergeHelper_QuatSwitch(
      0:Quat4f=$Output.UpdateMeshOrientation004.MeshOrientation
      1:Quat4f=$Output.UpdateMeshOrientation.MeshOrientation
      2:Quat4f=?
      3:Quat4f=?
      4:Quat4f=?
      5:Quat4f=?
      6:Quat4f=?
      7:Quat4f=?
      8:Quat4f=?
      9:Quat4f=?
      Index:NiagaraInt32=$Particles.StateIndex
      Num:MergeHelper_SwitcherEnum=::NewEnumerator0
     )
     Quaternion B:Quat4f=$Particles.BaseQuat
     Lerp Factor:NiagaraFloat=&FloatFromCurve(FloatCurve=DataInterface() {propertyValues {Curve {keys [<at 0.7 0>,<at 1.2 1>]}}})
     Enable Spherical Linear Interpolation:NiagaraBool=false
    )) {isSetParameters=true}
   /SetVariables(Particles.SpriteAlignment=&NormalizeVector(
     Vector To Normalize=&Subtract_Position(A=$Particles.CollisionPos, B=$Particles.Position)
     Safe Normalize=true
     Safe Fallback Threshold=0
     Safe Fallback Vector=<0 0 0.001>
    )) {isSetParameters=true, module=SetVariables_#12}
   &ScaleSpriteSize(
    Scale Factor=&Vector2DFromCurve(
     Vector2Curve=DataInterface() {propertyValues {XCurve {keys [<at 0 1 RCIM_Constant>,<at 0.5 0>]}, YCurve {keys [<at 0 0>,<at 0.2 1 RCIM_Constant>,<at 0.5 0>]}}}
     ScaleCurve=<1 1>
    )
    Scale Sprite Size Mode=::NewEnumerator2
   )
   &FN_Sprite_Attributes(
    Color=/Game/Effects/Niagara/DynamicInputs/TimeOfDay/DaySequenceTimeOfDayColor(Linear Color:LinearColor=<rgba 0.9989 0.9989 1 0.11>, Time Of Day Influence:NiagaraFloat=0.8)
    Create Color Attribute=true
    Module Output:ENiagara_AttributeSamplingApplyOutput=::NewEnumerator1
   )
   &FN_Sprite_Attributes(
    Color=&DaySequenceTimeOfDayColor(Linear Color:LinearColor=<rgba 0.9989 0.9989 1 0.6>, Time Of Day Influence:NiagaraFloat=0.8)
    Create Color Attribute=true
    Create SpriteSize Attribute=true
    Module Output:ENiagara_AttributeSamplingApplyOutput=::NewEnumerator1
    SpriteSize=&Multiply_Vector2D(
     A=&RandomRangeVector2D(Minimum=<25 2>, Maximum=<32 2>)
     B=&Vector2DFromCurve(
      Vector2Curve=DataInterface() {propertyValues {XCurve {keys [<at 0 1 RCIM_Constant>,<at 0.5 0>]}, YCurve {keys [<at 0 0>,<at 0.2 1 RCIM_Constant>,<at 0.5 0>]}}}
      ScaleCurve=<1 1>
     )
    )
   ) {module=FN_Sprite_Attributes001}
   &KillParticles(Kill Particles=&SetBoolByFloatComparison(
     A=$Particles.Age
     B=&MakeCustomFloatFromBool(Bool=$Particles.CollisionValid, True Float=5, False Float=2)
    ))
  ]
  rendererClasses [&NiagaraMeshRendererProperties, &NiagaraSpriteRendererProperties]
  renderers [
   {rendererClass=&NiagaraMeshRendererProperties, rendererIndex=0}
   {rendererClass=&NiagaraSpriteRendererProperties, rendererIndex=1}
   {rendererClass=&NiagaraSpriteRendererProperties, rendererIndex=2}
  ]
 }
 FakeBullets {
  emitterUpdateScript [
   &EmitterState(Life Cycle Mode=::NewEnumerator1, Loop Behavior=::NewEnumerator1, Loop Duration=1.5, UseLoopDelay=false)
   &SpawnRate(SpawnRate=40, Use Spawn Probability=false)
  ]
  particleSpawnScript [
   /SetVariables(Particles.RandomHeightOffset:NiagaraFloat=&RandomRangeFloat(Minimum=-20, Maximum=20)) {isSetParameters=true}
   &InitializeParticle(
    Color=&MergeHelper_ColorSwitch(0=<rgba 1 1 1 1>, 1=<rgba 1 1 1 1>, 2=<rgba 1 1 1 1>, 3=<rgba 1 1 1 1>, 4=<rgba 1 1 1 1>, Index=$Engine.QualityLevel, Num=::NewEnumerator3)
    Lifetime=&RandomRangeFloat(Minimum=0.9, Maximum=1.2)
    Mass Mode=::NewEnumerator1
    Mesh Scale Mode=::NewEnumerator4
    Position Offset=&MakeVector(X=0, Y=0, Z=&Add_Float(A=$Particles.RandomHeightOffset, B=$User.HeightStart))
    Sprite Rotation Angle=-0.25
    Sprite Rotation Mode=::NewEnumerator1
    Sprite Size=&RandomRangeVector2D(Minimum=<80 2>, Maximum=<120 3>)
    Sprite Size Mode=::NewEnumerator4
    UsePositionOffset=true
    Write Material Random=true
   )
   &ShapeLocation(Cylinder Height=20, MaskSpawnGroup=false, Override Seed=false, Shape Primitive=::NewEnumerator1, Surface Only=false, UseOwnerScale=true)
   /SetVariables(Particles.RandomDirection:Vector3f=&Add_Vector(A=<0 0 -1>, B=&RandomRangeVector(Minimum=<-0.1 -0.1 0>, Maximum=<0.1 0.1 0>))) {isSetParameters=true, module=SetVariables_#5}
   /SetVariables() {isSetParameters=true, module=SetVariables_#6}
   /SetVariables(Particles.StartPosition:NiagaraPosition=&Subtract_Position(A=$Particles.Position, B=&SimulationPosition())) {isSetParameters=true, module=SetVariables_#7}
   /SetVariables() {isSetParameters=true, module=SetVariables_#8}
   /SetVariables(Particles.NearPlayer:NiagaraBool=&SetBoolByFloatComparison(
     A=&VectorLength(Vector=&Multiply_Vector(
       A=&TransformPosition(Position=$Particles.Position, Source Space=::Simulation, Destination Space=::Local)
       B=<1 1 0>
      ))
     B=40
     Comparison Type=::NewEnumerator6
    )) {isSetParameters=true}
   &KillParticles(Kill Particles=$Particles.NearPlayer) {module=KillParticles001}
   /SetVariables(Particles.MeshOrientation=/Niagara/DynamicInputs/TypeConversions/MakeCustomQuaternionFromBool(
     Bool:NiagaraBool=&SetBoolByIntComparison(A=$Engine.QualityLevel, B=2, Comparison Type=::NewEnumerator6)
     True Quat:Quat4f=<0 0 0 1>
     False Quat:Quat4f=&MakeQuaternion(
      Quaternion From=5
      Axis=?
      Angle=?
      Yaw=?
      Pitch=?
      Roll=?
      X=?
      Y=?
      Z=?
      From=?
      To=?
      Stabilization Mode:ENiagara_OrientVectorCount=::NewEnumerator0
      Side Direction:Vector3f=?
      Up Direction:Vector3f=?
      Angle Type=::NewEnumerator0
      Coordinate Space=?
      Normalize:NiagaraBool=true
     )
    )) {enabled=false, isSetParameters=true}
  ]
  particleUpdateScript [
   &ParticleState(Kill Particles When Lifetime Has Elapsed=false, Loop Particles Lifetime=true)
   /SetVariables(Particles.CollisionPos:NiagaraPosition=&MakeCustomVectorFromBool(
     Bool=$Particles.CollisionValid
     True Vector=Linked() {linkedVariable {name=Particles.CollisionPos, type=NiagaraPosition}}
     False Vector=&AddVectorToPosition(
      Position=&AddVectorToPosition(Position=$Particles.Initial.StartPosition, Vector=&SimulationPosition())
      Vector=&Multiply_VectorByFloat(Vector=<0 0 -1>, Float=&Add_Float(A=$Particles.RandomHeightOffset, B=$User.HeightStart))
     )
    )) {isSetParameters=true, module=SetVariables_#9}
   /SetVariables(Particles.Position=&LerpPosition(
     A=&AddVectorToPosition(
      Position=&SimulationPosition()
      Vector=Linked() {linkedVariable {name=Particles.Initial.StartPosition, type=NiagaraPosition}}
     )
     B=$Particles.CollisionPos
     Alpha=&FloatFromCurve(
      FloatCurve=DataInterface() {propertyValues {Curve {
         defaultValue=1
         keys [
          {arriveTangent=0.4591, interpMode=RCIM_Cubic, leaveTangent=0.4591, tangentMode=RCTM_User, time=0, value=0}
          {arriveTangent=2.0648, interpMode=RCIM_Cubic, leaveTangent=2.0648, tangentMode=RCTM_User, time=1, value=1}
         ]
        }}}
      CurveIndex=&Multiply_Float(A=$Particles.NormalizedAge, B=1)
     )
    )) {isSetParameters=true, module=SetVariables_#10}
   &ScaleMeshSize(Scale Factor=&Multiply_Vector(
     A=&VectorFromCurve(VectorCurve=DataInterface() {propertyValues {
        XCurve {keys [<at 0 1>,<at 1 1>,<at 1 1>]}
        YCurve {keys [<at 0 1 RCIM_Cubic>, <at 0.45 1.3 RCIM_Cubic>, {arriveTangent=-1.6667, time=0.5, value=1}, <at 1 1 RCIM_Cubic>, <at 1 1 RCIM_Cubic>]}
        ZCurve {keys [<at 0 1>,<at 1 1>,<at 1 1>]}
       }})
     B=&VectorFromCurve(
      VectorCurve=DataInterface() {propertyValues {
        XCurve {keys [
          <at 0 0 RCIM_Cubic>
          {arriveTangent=1.0526, leaveTangent=1.0526, time=0.1, value=1}
          {arriveTangent=-1.1111, interpMode=RCIM_Cubic, leaveTangent=-1.1111, time=0.95, value=1}
          <at 1 0 RCIM_Cubic>
         ]}
        YCurve {keys [<at 0 0>,<at 0.1 1>,<at 1 1>]}
        ZCurve {keys [
          <at 0 0 RCIM_Cubic>
          {arriveTangent=1.0526, leaveTangent=1.0526, time=0.1, value=1}
          {arriveTangent=-1.1111, interpMode=RCIM_Cubic, leaveTangent=-1.1111, time=0.95, value=1}
          <at 1 0 RCIM_Cubic>
         ]}
       }}
      Scale Curve=&VectorFromFloat(Value=&CameraFade(
        Camera Query:NiagaraDataInterfaceCamera=DataInterface()
        Fade Offset:NiagaraFloat=50
        Fade Distance:NiagaraFloat=100
        Use Owner Position:NiagaraBool=true
       ))
     )
    )) {module=ScaleMeshSize001}
   /SetVariables(Particles.StateIndex:NiagaraInt32=&MakeIntFromFloat(
     Float=&FloatFromCurve(FloatCurve=DataInterface() {propertyValues {Curve {keys [<at 0 0 RCIM_Constant>,<at 0.5 1 RCIM_Constant>]}}})
     Conversion Mode=::NewEnumerator1
    )) {isSetParameters=true, module=SetVariables_#11}
   /SetVariables(Particles.SpriteAlignment=&NormalizeVector(
     Vector To Normalize=&Subtract_Position(A=$Particles.CollisionPos, B=$Particles.Position)
     Safe Normalize=true
     Safe Fallback Threshold=0
     Safe Fallback Vector=<0 0 0.001>
    )) {isSetParameters=true, module=SetVariables_#12}
   &UpdateMeshOrientation(
    Facing Direction=&Lerp_Vector(
     A=&RandomRangeVector(Minimum=<-1 -1 -1>, Maximum=<1 1 1>)
     B=&NormalizeVector(
      Vector To Normalize=&RotateVectorByQuaternion(
       Vector To Rotate=&RandomRangeVector(Minimum=<-1 -1 -1>, Maximum=<1 1 1>)
       Quaternion=&MakeQuaternion(
        Quaternion From=2
        Axis=&RandomRangeVector(Minimum=<-1 -1 -1>, Maximum=<1 1 1>)
        Angle=$Emitter.Age
        X=&Multiply_Float(A=$Emitter.Age, B=&RandomRangeFloat(Minimum=-100, Maximum=100))
        Y=&Multiply_Float(A=$Emitter.Age, B=&RandomRangeFloat(Minimum=-100, Maximum=100))
        Z=&Multiply_Float(A=$Emitter.Age, B=&RandomRangeFloat(Minimum=-100, Maximum=100))
        From=?
        To=?
        Stabilization Mode:ENiagara_OrientVectorCount=::NewEnumerator0
        Side Direction:Vector3f=?
        Up Direction:Vector3f=?
        Angle Type=::NewEnumerator0
        Normalize:NiagaraBool=true
       )
      )
      Safe Normalize=true
      Safe Fallback Threshold=0
      Safe Fallback Vector=<0 0 0.001>
     )
     Alpha=&FloatFromCurve(
      FloatCurve=DataInterface() {propertyValues {Curve {
         defaultValue=1
         keys [<at 0 0 RCIM_Cubic>, {arriveTangent=1.6372, interpMode=RCIM_Cubic, leaveTangent=1.6372, tangentMode=RCTM_User, time=1, value=1}]
        }}}
      CurveIndex=$Emitter.Age
     )
    )
    Facing Interpolation=false
    Orientation Method=::NewEnumerator1
    Override Quat=false
    TransformFacing=false
    TransformUp=false
    Up Direction:Vector3f=$System.CameraUp
   ) {module=UpdateMeshOrientation002}
   /SetVariables(Particles.BaseQuat:Quat4f=$Output.UpdateMeshOrientation004.MeshOrientation) {enabled=false, isSetParameters=true}
   &ScaleSpriteSize(
    Scale Factor=&Vector2DFromCurve(
     Vector2Curve=DataInterface() {propertyValues {XCurve {keys [<at 0 1 RCIM_Constant>]}, YCurve {keys [<at 0 0>,<at 0.2 1 RCIM_Constant>]}}}
     ScaleCurve=<1 1>
    )
    Scale Sprite Size Mode=::NewEnumerator2
   )
   &FN_Sprite_Attributes(
    Color=&DaySequenceTimeOfDayColor(Linear Color:LinearColor=<rgba 0.9989 0.9989 1 0.11>, Time Of Day Influence:NiagaraFloat=0.8)
    Create Color Attribute=true
    Module Output:ENiagara_AttributeSamplingApplyOutput=::NewEnumerator1
   )
   &FN_Sprite_Attributes(
    Color=&DaySequenceTimeOfDayColor(Linear Color:LinearColor=<rgba 0.9989 0.9989 1 0.6>, Time Of Day Influence:NiagaraFloat=0.8)
    Create Color Attribute=true
    Create SpriteSize Attribute=true
    Module Output:ENiagara_AttributeSamplingApplyOutput=::NewEnumerator1
    SpriteSize=&Multiply_Vector2D(
     A=&RandomRangeVector2D(Minimum=<25 2>, Maximum=<32 2>)
     B=&Vector2DFromCurve(
      Vector2Curve=DataInterface() {propertyValues {XCurve {keys [<at 0 1 RCIM_Constant>]}, YCurve {keys [<at 0 0>,<at 0.2 1 RCIM_Constant>]}}}
      ScaleCurve=<1 1>
     )
    )
   ) {module=FN_Sprite_Attributes001}
  ]
  rendererClasses [&NiagaraMeshRendererProperties, &NiagaraSpriteRendererProperties]
  renderers [
   {rendererClass=&NiagaraMeshRendererProperties, rendererIndex=0}
   {rendererClass=&NiagaraSpriteRendererProperties, rendererIndex=1}
   {rendererClass=&NiagaraSpriteRendererProperties, rendererIndex=2}
   {rendererClass=&NiagaraMeshRendererProperties, rendererIndex=3}
  ]
 }
 FlashFlares {
  emitterUpdateScript [
   &EmitterState()
   &SpawnPerFrame(Spawn:NiagaraBool=$System.DeathSpawn, Use Spawn Probability=false)
  ]
  particleSpawnScript [
   &InitializeParticle(
    Color=<rgba 2 0.9225 0.25 1>
    Lifetime=0.2
    Position=$System.DeathPosition
    Position Mode=::NewEnumerator1
    Position Offset=&MaskVectorBySpawnGroup(Masked Vector=<0 0 0>, Passthrough Vector=<0 0 0>)
    Position Offset Coordinate Space=::World
    Sprite Size=&RandomRangeVector2D(Minimum=<80 30>, Maximum=<30 15>)
    Sprite Size Mode=::NewEnumerator4
    UsePositionOffset=true
    Write Material Random=true
   ) {module=InitializeParticle001}
   &CameraOffset(Camera Offset Amount=10) {module=CameraOffset001}
   &PartitionParticles()
   &DynamicMaterialParameters(
    Index 0 Param 1=0.2
    Module.Param0WriteEnabled=true
    Module.Param1WriteEnabled=true
    Module.Param2WriteEnabled=true
    Module.Param3WriteEnabled=true
   )
  ]
  particleUpdateScript [
   &ParticleState()
   /SetVariables(
    Particles.SparkColor:LinearColor=&ColorFromCurve(ColorCurve=DataInterface() {propertyValues {
       AlphaCurve {keys [<at 0 1 RCIM_Cubic>,<at 0.6826 0 RCIM_Cubic>]}
       BlueCurve {keys [<at 0 0.0649 RCIM_Cubic>,<at 1 0 RCIM_Cubic>]}
       GreenCurve {keys [<at 0 1.9579 RCIM_Cubic>,<at 1 0.3992 RCIM_Cubic>]}
       RedCurve {keys [<at 0 5 RCIM_Cubic>,<at 1 1 RCIM_Cubic>]}
      }})
    Particles.SparkSize:Vector2f=&Vector2DFromFloat(Value=&FloatFromCurve(
      FloatCurve=DataInterface() {propertyValues {Curve {keys [{arriveTangent=2.3976, interpMode=RCIM_Cubic, leaveTangent=2.3976, tangentMode=RCTM_User, time=0, value=0}, <at 1 1 RCIM_Cubic>]}}}
      Scale Curve=10
     ))
   ) {isSetParameters=true}
   &SubUVAnimation(
    Random Start Frame=true
    "Sprite Renderer "=DataInterface() {propertyValues {SpriteRenderer=&self:FlashFlares_0.NiagaraSpriteRendererProperties_1}}
    SubUV Lookup Index Scale=5.7547
    UseEndFrame=false
    UseStartFrame=false
   )
   &ScaleColor(
    Scale Alpha=&FloatFromCurve(FloatCurve=DataInterface() {propertyValues {Curve {keys [<at 0 1>,<at 1 0>]}}}, Scale Curve=4)
    ScaleA=true
    ScaleRGB=true
   )
  ]
  rendererClasses [&NiagaraSpriteRendererProperties]
  renderers [
   {rendererClass=&NiagaraSpriteRendererProperties}
   {rendererClass=&NiagaraSpriteRendererProperties, rendererIndex=1}
  ]
 }
 Sparks {
  emitterUpdateScript [
   &EmitterState()
   &SpawnPerFrame(
    Spawn:NiagaraBool=$System.DeathSpawn
    Spawn Count:NiagaraInt32=&RandomRangeInt(Minimum=3, Maximum=5)
    Use Spawn Probability=false
   )
  ]
  particleSpawnScript [
   &InitializeParticle(
    Color Mode=::NewEnumerator0
    Lifetime=&RandomRangeFloat(Minimum=0.18, Maximum=0.25)
    Position=$System.DeathPosition
    Position Mode=::NewEnumerator1
    Sprite Rotation Mode=::NewEnumerator2
    Sprite Size Mode=::NewEnumerator1
    Uniform Sprite Size Max=2
    Uniform Sprite Size Min=1
    UsePositionOffset=false
    Write Material Random=true
   )
   &AddVelocity(
    Cone Angle=60
    Cone Axis=<0 0 1>
    Module.UseConeFalloff=false
    Module.UseVelDistribution=false
    Override Seed=false
    Velocity Mode=::NewEnumerator2
    Velocity Speed=&RandomRangeFloat(Minimum=10, Maximum=200)
   ) {module=AddVelocity001}
   /SetVariables(Particles.Position=&AddVectorToPosition(
     Position=$Particles.Position
     Vector=&Multiply_VectorByFloat(
      Vector=&NormalizeVector(Vector To Normalize=$Particles.Velocity, Safe Normalize=true, Safe Fallback Threshold=0, Safe Fallback Vector=<0 0 0.001>)
      Float=&RandomRangeFloat(Minimum=5, Maximum=10)
     )
    )) {isSetParameters=true}
  ]
  particleUpdateScript [
   &ParticleState()
   /SetVariables(
    Particles.Flicker:NiagaraFloat=&Waveform("Waveform [1]"=::NewEnumerator0, "[1] Frequency"=75, "[1] Amplitude Min/Max"=<0 2>)
    Particles.Flicker Size:NiagaraFloat=&Waveform("Waveform [1]"=::NewEnumerator0, "[1] Frequency"=75, "[1] Amplitude Min/Max"=<0.5 1>)
    Particles.Flicker Value:NiagaraFloat=&Waveform("Waveform [1]"=::NewEnumerator0, "[1] Frequency"=75, "[1] Amplitude Min/Max"=<2 8>)
   ) {isSetParameters=true}
   &ScaleSpriteSize(
    Scale Factor=/Niagara/DynamicInputs/Add/Add_Vector2(
     A:Vector2f=&Vector2DFromCurve(
      Vector2Curve=DataInterface() {propertyValues {XCurve {keys [<at 0 0>,<at 0 0>,<at 0.1 1>,<at 0.8 0>,<at 1 0>]}, YCurve {keys [<at 0 0>,<at 0.1 1>,<at 0.8 0>,<at 1 0>]}}}
      ScaleCurve=&RandomRangeVector2D(Minimum=<-0.3 5>, Maximum=<-1 10>)
     )
     B:Vector2f=&Vector2DFromCurve(
      Vector2Curve=DataInterface() {propertyValues {XCurve {keys [<at 0 0>,<at 0.2 1>,<at 1 0>]}, YCurve {keys [<at 0 0>,<at 0.2 1>,<at 1 0>]}}}
      ScaleCurve=<1 1>
     )
    )
    Scale Sprite Size Mode=::NewEnumerator2
   ) {module=ScaleSpriteSize001}
   &ScaleColor(
    Scale Mode=::NewEnumerator1
    Scale RGBA=&Multiply_Vector4(
     A=&MakeVector4_FromLinearColor(LinearColor=&ColorFromCurve(ColorCurve=DataInterface() {propertyValues {
         AlphaCurve {keys [<at 0.122 1>]}
         BlueCurve {keys [<at 0 1.0417>,<at 0.0705 0.9911>,<at 0.2273 0>,<at 1 0>]}
         GreenCurve {keys [<at 0 7.4334>,<at 0.0705 10.4768>,<at 0.2273 9.1355>,<at 1 3.6718>]}
         RedCurve {keys [<at 0 20>,<at 0.0705 16>,<at 0.2273 13>,<at 1 9>]}
        }}))
     B=<1 1 1 1>
    )
    ScaleRGBA=true
   )
   &Drag(Drag=&RandomRangeFloat(Minimum=5, Maximum=10), Use Linear Drag=true, Use Rotational Drag=true)
   &SolveForcesAndVelocity(Clamp Velocity=false, Limit Acceleration=false)
   /SetVariables() {isSetParameters=true}
   &FN_Sprite_Attributes(
    Color=&MultiplyLinearColor(
     Linear Color=<rgba 1 0.1971 0 0.5>
     Scale Factor=&MakeFloat4FromVectorAndFloat(
      "Vector (RGB)"=<1 1 1>
      "Float (Alpha)"=&FloatFromCurve(FloatCurve=DataInterface() {propertyValues {Curve {keys [<at 0.2 1>,<at 1 0>]}}}, Scale Curve=0.8)
     )
    )
    Create Color Attribute=true
    Create SpriteSize Attribute=true
    SpriteSize=&Multiply_Vector2D(A=<10 2>, B=$Particles.SpriteSize)
    Write to Module Namespace=true
   )
  ]
  rendererClasses [&NiagaraSpriteRendererProperties]
  renderers [
   {rendererClass=&NiagaraSpriteRendererProperties}
   {rendererClass=&NiagaraSpriteRendererProperties, rendererIndex=1}
  ]
 }
}
summary {
 systemName=NS_Emote_AlpacaLean_BulletRain
 emitters [
  {emitterName=FlashFlares, rendererClasses [&NiagaraSpriteRendererProperties]}
  {emitterName=Sparks, rendererClasses [&NiagaraSpriteRendererProperties]}
  {emitterName=Bullets, rendererClasses [&NiagaraMeshRendererProperties, &NiagaraSpriteRendererProperties]}
  {emitterName=FakeBullets, rendererClasses [&NiagaraMeshRendererProperties, &NiagaraSpriteRendererProperties]}
 ]
 userVariables [
  User.bIsFrontEnd {type=NiagaraBool}
  User.bIsFrontEndPreview {type=NiagaraBool}
  User.CollisionCheck {type=NiagaraFloat, defaultValue=0.45, description=""}
  User.HeightStart {type=NiagaraFloat, defaultValue=230, description=""}
 ]
}
